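# Control Rig for the Pennant Seas One character's face accessory part that procedurally swings a group of hair bones as the head rotates.
# The forward solve reads the local head rotation as Euler Z and remaps a negative angle range into a clamped blend weight.
# That weight drives a fixed Z rotation on master_dyn_hair_movement_jnt, carrying its child bones dyn_hair_01 to 05, while dyn_hair_06 to 09 stay directly under the head.
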ControlRigBlueprint /BRCosmetics/Characters/Player/Male/Medium/Bodies/M_MED_Pennant_Seas_One/Meshes/Parts/M_MED_Pennant_Seas_One_ASN_FaceAcc_CtrlRig
graphs [
 RigVMFunctionLibrary {links [], nodes []}
 RigVMModel {nodes [
   RigUnit_BeginExecution = &RigUnit_BeginExecution()
   n1 = &RigUnit_GetTransform(Item=(Type=Bone,Name="head"), Space=LocalSpace)
   &RigUnit_ModifyTransforms(
    ItemToModify=((Item=(Type=Bone,Name="master_dyn_hair_movement_jnt"),Transform=(Rotation=(X=0,Y=0,Z=0,W=1),Translation=(X=0,Y=0,Z=0),Scale3D=(X=1,Y=1,Z=1))))
    Weight=&RigVMFunction_MathFloatRemap(
     SourceMaximum=-35
     SourceMinimum=-10
     bClamp=true
     Value=&RigVMFunction_MathQuaternionToEuler(RotationOrder=ZYX, Value=@n1.Transform.Rotation).Result.Z
    ).Result
    ItemToModify.0.Transform.Rotation=&RigVMFunction_MathQuaternionFromEuler(Euler=(X=0,Y=0,Z=-50)).Result
    ExecutePin=@RigUnit_BeginExecution.ExecutePin
   )
  ]}
]
hierarchy [
 root {loc=<0 0 0>, rot=<0 0 0 1>}
 pelvis {parent=root, loc=<0 -0.7071 93.1191>, rot=<0.0127 0.707 -0.0127 -0.707>}
 spine_01 {parent=pelvis, loc=<6.7494 -0.6388 0>, rot=<0 0 0.034 -0.9994>}
 spine_02 {parent=spine_01, loc=<7.2518 -0.035 0>, rot=<0 0 0.0238 -0.9997>}
 spine_03 {parent=spine_02, loc=<6.0104 0 0>, rot=<0 0 0.0149 -0.9999>}
 spine_04 {parent=spine_03, loc=<8.1024 0 0>, rot=<0 0 -0.0132 -0.9999>}
 spine_05 {parent=spine_04, loc=<10.0146 0 0>, rot=<0 0 -0.0015 -1>}
 neck_01 {parent=spine_05, loc=<11.3138 0.1529 0>, rot=<0 0 -0.2003 -0.9797>}
 neck_02 {parent=neck_01, loc=<4.5627 0 0>, rot=<0 0 0.0504 -0.9987>}
 head {parent=neck_02, loc=<5.3914 0 0>, rot=<0 0 0.115 -0.9934>}
 dyn_hair_06 {parent=head, loc=<14.1994 -2.6767 1.3761>, rot=<0.0029 -0.7071 -0.0029 -0.7071>}
 dyn_hair_07 {parent=head, loc=<14.1994 -2.6767 -2.6516>, rot=<0.0416 -0.6892 -0.0137 -0.7232>}
 dyn_hair_08 {parent=head, loc=<15.5437 4.7891 -0.6416>, rot=<0.0029 -0.7071 -0.0029 -0.7071>}
 dyn_hair_09 {parent=head, loc=<11.9807 8.1572 3.701>, rot=<0.0029 -0.7071 -0.0029 -0.7071>}
 master_dyn_hair_movement_jnt {parent=head, loc=<4.3981 -4.9007 0>, rot=<0.1439 -0.9896 0 0>}
 dyn_hair_01 {parent=master_dyn_hair_movement_jnt, loc=<2.2885 1.8993 -2.9834>, rot=<-0.0989 0.7002 -0.0989 -0.7002>}
 dyn_hair_02 {parent=master_dyn_hair_movement_jnt, loc=<0.9581 3.6363 -5.0013>, rot=<-0.0989 0.7002 -0.0989 -0.7002>}
 dyn_hair_03 {parent=master_dyn_hair_movement_jnt, loc=<2.4529 1.3287 -1.0035>, rot=<-0.0989 0.7002 -0.0989 -0.7002>}
 dyn_hair_04 {parent=master_dyn_hair_movement_jnt, loc=<2.5904 0.8518 1.7972>, rot=<-0.0989 0.7002 -0.0989 -0.7002>}
 dyn_hair_05 {parent=master_dyn_hair_movement_jnt, loc=<1.4726 1.9957 3.9335>, rot=<-0.0989 0.7002 -0.0989 -0.7002>}
]
